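# Material function that outputs the UVs for the hologram effect's scanlines.
# With the Scanline_Use_Custom_Direction_Vector static switch off, it passes viewport UVs through.
# With it on, screen position is centred on the particle's projected clip-space position, then rotated using a basis built from a world-space direction transformed to view space and its perpendicular.
# That direction comes per particle from Dynamic Parameter index 1.
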MaterialFunction /Game/Tools/3DHB/Shader/Base/MF_3DHB_HologramFX_UVForScanline
graph {nodes [
  n1 = ScreenPosition()
  n2 = ParticlePositionWS()
  n3 = Subtract(
   A=ConstantBiasScale(None=@n1.ViewportUV) {bias=-0.5, scale=2}
   B=Multiply(
    A=MaterialFunctionCall(World Position=@n2) {materialFunction=&TransformToClipSpace}.Clip Space XY
    B=Constant2Vector() {g=-1, r=1}
   )
  )
  n4 = DynamicParameter() {
   parameterIndex=1
   defaultValue=<rgba 0 -1 1 1>
   paramNames [Scanline UVs Direction Vector WS X, Scanline UVs Direction Vector WS Y, Scanline UVs Direction Vector WS Z, Param4]
  }
  n5 = Transform(None=Normalize(VectorInput=@n4.RGB)) {transformSourceType=TRANSFORMSOURCE_World, transformType=TRANSFORM_View}
  n6 = Multiply(A=ComponentMask(None=@n3) {r=true}, B=Normalize(VectorInput=ComponentMask(None=@n5) {g=true, r=true}))
  n7 = CrossProduct(A=@n5, B=Constant3Vector() {constant=<rgba 0 0 1 0>})
  n8 = Multiply(A=ComponentMask(None=@n3) {g=true}, B=Normalize(VectorInput=ComponentMask(None=@n7) {g=true, r=true}))
  n9 = StaticSwitchParameter(True=ConstantBiasScale(None=Add(A=@n6, B=@n8)), False=@n1.ViewportUV) {
   desc="Feed World Space direction vector via Dynamic Parameter index ##1"
   group=Base_HoloFX_Scanline
   parameterName=Scanline_Use_Custom_Direction_Vector
   sortPriority=-1
  }
  FunctionOutput(None=@n9) {outputName=UVForScanline}
 ]}
